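; Kitchen wood stove building prop (variant B) that tints its long and short stove pipe meshes to match the building's texture data.
; The construction script runs the parent, then reads the building texture data array at a local index.
; A valid index applies that entry's tint color to both pipes as secondary meshes; otherwise both are set to white.
; BeginPlay, overlap and Tick events are empty.
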(event EventBeginPlay)

(event Collision|EventActorBeginOverlap
  (params (OtherActor :type Actor%20Object%20Reference)))

(event EventTick
  (params (DeltaSeconds :type Float%20%28single-precision%29)))

(fn ConstructionScript
  (bind _l_tdindex (LocalVariables|GetL_TDIndex))
  (bind _sm_kitchen_stovewoodpipelong_a (Variables|Default|GetSM_Kitchen_StoveWoodPipeLong_A))
  (bind _sm_kitchen_stovewoodpipeshort_a (Variables|Default|GetSM_Kitchen_StoveWoodPipeShort_A))
  (|CallParentFunction)
  (bind _outtexturedataarray (Building|GetBuildingTextureData))
  (if (Utilities|Array|IsValidIndex _outtexturedataarray _l_tdindex)
    (Default|SetColoronSecondaryMesh (Math|Conversions|ToLinearColor(Color) (Variables|/Script/FortniteGame.BuildingTextureData|GetTintColor (Utilities|Array|Get(acopy) _outtexturedataarray _l_tdindex))) _sm_kitchen_stovewoodpipelong_a)
    (Default|SetColoronSecondaryMesh (Math|Conversions|ToLinearColor(Color) (Variables|/Script/FortniteGame.BuildingTextureData|GetTintColor (Utilities|Array|Get(acopy) _outtexturedataarray _l_tdindex))) _sm_kitchen_stovewoodpipeshort_a)
    (else
      (Default|SetColoronSecondaryMesh "(R=1.000000,G=1.000000,B=1.000000,A=1.000000)" _sm_kitchen_stovewoodpipelong_a)
      (Default|SetColoronSecondaryMesh "(R=1.000000,G=1.000000,B=1.000000,A=1.000000)" _sm_kitchen_stovewoodpipeshort_a))))
